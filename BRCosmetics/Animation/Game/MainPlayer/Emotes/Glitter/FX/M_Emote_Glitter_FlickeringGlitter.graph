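# Translucent unlit Niagara sprite material for the Glitter emote, built to give glitter particles flickering, varied brightness.
# Camera-relative particle position is hashed into UVs for a Perlin normal texture, so the sample changes as particle or camera moves.
# Red scales particle color before the emissive adjustment function; green drives a separate HueShift.
# A UseTexture parameter blends a procedural circle mask with a sparkle texture.
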Material /BRCosmetics/Animation/Game/MainPlayer/Emotes/Glitter/FX/M_Emote_Glitter_FlickeringGlitter
graph {
 properties {
  bAutomaticallySetUsageInEditor=false
  bUsedWithNiagaraSprites=true
  blendMode=BLEND_Translucent
  pixelDepthOffsetMode=PDOM_Legacy
  shadingModel=MSM_Unlit
  usedShadingModels=Unlit
 }
 nodes [
  Constant() {r=0.1}
  n1 = Subtract(A=TextureCoordinate()) {constB=0.5}
  LinearInterpolate(
   A=Step(Y=Length(None=@n1) {is=class MaterialExpressionLength}) {constX=0.4}
   B=TextureSample() {texture=/Game/Effects/Fort_Effects/Textures/Sparkle/T_Sparkle_01}.R
   Alpha=ScalarParameter() {parameterName=UseTexture}
  )
  n2 = Multiply(A=ParticlePositionWS() {originType=CameraRelative}) {constB=5}
  n3 = Add(A=Multiply(A=ComponentMask(None=@n2) {r=true}) {constB=2.5121}, B=Multiply(A=ComponentMask(None=@n2) {g=true}) {constB=1.6125})
  n4 = Multiply(A=ComponentMask(None=@n2) {b=true}) {constB=3.178}
  n5 = TextureSample(UVs=Frac(None=Multiply(A=@n3, B=@n4))) {samplerType=SAMPLERTYPE_Normal, texture=/BRCosmetics/Characters/Player/Male/Large/Bodies/M_LRG_Clash_V/Textures/Simple_Perlin_N}
  n6 = Multiply(A=LinearInterpolate(Alpha=@n5.R) {constA=0.25, constB=1.5}, B=ParticleColor().RGB)
  MaterialFunctionCall(Hue Shift Percentage=LinearInterpolate(Alpha=@n5.G) {constA=0.1, constB=0.3}, Texture=@n6) {materialFunction=&HueShift}
  MaterialFunctionCall(Emissive=@n6) {materialFunction=&MF_FX_EmissiveAdjustments}
 ]
}
