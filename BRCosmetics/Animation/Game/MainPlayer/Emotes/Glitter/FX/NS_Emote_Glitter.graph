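# Niagara effect for the Glitter emote that throws a one-shot upward cone of white, blue and purple sprite confetti and sparks.
# A GPU burst, its smaller CPU twin BurstMedLowOnly, speed-stretched streaks, residual bits, SubUV squares and twinkling sparks form six independent emitters.
# Each tints a random palette color by a system time-of-day color.
# Drag, gravity and curl noise shape most motion, while size curves handle fades and flicker.
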NiagaraSystem /BRCosmetics/Animation/Game/MainPlayer/Emotes/Glitter/FX/NS_Emote_Glitter
dependencies {
 usedDataInterfaces [
  &NiagaraDataInterfaceArrayColor
  &NiagaraDataInterfaceCurve
  &NiagaraDataInterfaceSpriteRendererInfo
 ]
 usedDynamicInputs [&FNTimeOfDayColor]
 usedModules [
  &SystemState
  &self:SystemSpawnScript.SystemScriptSource.SystemScriptGraph.NiagaraNodeAssignment_0.SetVariables_#1
 ]
}
emitters {
 BurstMedLowOnly {
  emitterUpdateScript [
   &EmitterState()
   &SpawnBurst_Instantaneous(Spawn Count=35, Use Spawn Probability=false, UseLoopCountLimit=false)
  ]
  particleSpawnScript [
   &InitializeParticle(
    Color=&MultiplyLinearColor(
     Linear Color=&SelectLinearColorFromArray(
      Array Sampling Mode=::NewEnumerator0
      Direct Array Index=?
      Color Selection Array=DataInterface() {propertyValues {ColorData [<rgba 0.849 0.849 0.849 1>, <rgba 0.1674 0.4847 2 1>, <rgba 0.4995 0.1319 1.5 1>]}}
      Random Seed=0
     )
     Scale Factor=&MakeVector4_FromLinearColor(LinearColor=$System.TOD)
    )
    Lifetime Max=1
    Lifetime Min=0.4
    Lifetime Mode=::NewEnumerator1
    Sprite Size Mode=::NewEnumerator1
    Uniform Sprite Size Max=2
    Uniform Sprite Size Min=1
    UsePositionOffset=false
    Write Material Random=true
   )
   &ShapeLocation(MaskSpawnGroup=false, Override Seed=false, Sphere Radius=2, UseOwnerScale=true) {enabled=false}
   &AddVelocity(
    Cone Angle=30
    Cone Axis=<0 0 1>
    Module.UseConeFalloff=false
    Module.UseVelDistribution=false
    Override Seed=false
    Rotation Coordinate Space=::Simulation
    Velocity Mode=::NewEnumerator2
    Velocity Speed=&RandomRangeFloat(Minimum=250, Maximum=600)
   )
   &AddVelocity(Velocity=<0 -100 0>) {enabled=false, module=AddVelocity001}
  ]
  particleUpdateScript [
   &Drag(Drag=3, Use Linear Drag=true, Use Rotational Drag=true)
   &GravityForce(Gravity=&Multiply_VectorByFloat(
     Vector=<1 1 -1200>
     Float=&FloatFromCurve(FloatCurve=DataInterface() {propertyValues {Curve {keys [<at 0 1>, {interpMode=RCIM_Cubic, leaveTangent=-3.9842, tangentMode=RCTM_Break, time=0.3, value=1}, <at 0.5 0.2>, <at 1 0>]}}})
    ))
   &SolveForcesAndVelocity(Clamp Velocity=false, Limit Acceleration=false)
   &ParticleState()
   &ScaleSpriteSize(Uniform Curve Sprite Scale=DataInterface() {propertyValues {Curve {keys [
        {interpMode=RCIM_Cubic, leaveTangent=6.8368, tangentMode=RCTM_Break, time=0, value=0}
        <at 0.2 1>
        {interpMode=RCIM_Cubic, leaveTangent=-4.4769, tangentMode=RCTM_Break, time=0.6, value=1}
        {tangentMode=RCTM_User, time=1.0114, value=0.0706}
       ]}}})
  ]
  rendererClasses [&NiagaraSpriteRendererProperties]
  renderers [{rendererClass=&NiagaraSpriteRendererProperties}]
 }
 ConfettiBurst {
  simTarget=GPUComputeSim
  emitterUpdateScript [
   &EmitterState()
   &SpawnBurst_Instantaneous(Spawn Count=70, Use Spawn Probability=false, UseLoopCountLimit=false)
  ]
  particleSpawnScript [
   &InitializeParticle(
    Color=&MultiplyLinearColor(
     Linear Color=&SelectLinearColorFromArray(
      Array Sampling Mode=::NewEnumerator0
      Direct Array Index=?
      Color Selection Array=DataInterface() {propertyValues {ColorData [<rgba 0.849 0.849 0.849 1>, <rgba 0.1674 0.4847 2 1>, <rgba 0.4995 0.1319 1.5 1>]}}
      Random Seed=0
     )
     Scale Factor=&MakeVector4_FromLinearColor(LinearColor=$System.TOD)
    )
    Lifetime Max=1
    Lifetime Min=0.4
    Lifetime Mode=::NewEnumerator1
    Sprite Size Mode=::NewEnumerator1
    Uniform Sprite Size Max=1.5
    Uniform Sprite Size Min=0.7
    UsePositionOffset=false
    Write Material Random=true
   )
   &ShapeLocation(MaskSpawnGroup=false, Override Seed=false, Sphere Radius=2, UseOwnerScale=true) {enabled=false}
   &AddVelocity(
    Cone Angle=30
    Cone Axis=<0 0 1>
    Module.UseConeFalloff=false
    Module.UseVelDistribution=false
    Override Seed=false
    Rotation Coordinate Space=::Simulation
    Velocity Mode=::NewEnumerator2
    Velocity Speed=&RandomRangeFloat(Minimum=250, Maximum=600)
   )
   &AddVelocity(Velocity=<0 -100 0>) {enabled=false, module=AddVelocity001}
  ]
  particleUpdateScript [
   &Drag(Drag=3, Use Linear Drag=true, Use Rotational Drag=true)
   &GravityForce(Gravity=&Multiply_VectorByFloat(
     Vector=<1 1 -1200>
     Float=&FloatFromCurve(FloatCurve=DataInterface() {propertyValues {Curve {keys [<at 0 1>, {interpMode=RCIM_Cubic, leaveTangent=-3.9842, tangentMode=RCTM_Break, time=0.3, value=1}, <at 0.5 0.2>, <at 1 0>]}}})
    ))
   &SolveForcesAndVelocity(Clamp Velocity=false, Limit Acceleration=false)
   &ParticleState()
   &ScaleSpriteSize(Uniform Curve Sprite Scale=DataInterface() {propertyValues {Curve {keys [
        {interpMode=RCIM_Cubic, leaveTangent=6.8368, tangentMode=RCTM_Break, time=0, value=0}
        <at 0.2 1>
        {interpMode=RCIM_Cubic, leaveTangent=-4.4769, tangentMode=RCTM_Break, time=0.6, value=1}
        {tangentMode=RCTM_User, time=1.0114, value=0.0706}
       ]}}})
  ]
  rendererClasses [&NiagaraSpriteRendererProperties]
  renderers [{rendererClass=&NiagaraSpriteRendererProperties}]
 }
 ConfettiResidual {
  emitterUpdateScript [
   &EmitterState()
   &SpawnBurst_Instantaneous(Spawn Count=12, Use Spawn Probability=false, UseLoopCountLimit=false)
  ]
  particleSpawnScript [
   &InitializeParticle(
    Color=&MultiplyLinearColor(
     Linear Color=&SelectLinearColorFromArray(
      Array Sampling Mode=::NewEnumerator0
      Direct Array Index=?
      Color Selection Array=DataInterface() {propertyValues {ColorData [<rgba 0.849 0.849 0.849 1>, <rgba 0.1674 0.4847 2 1>, <rgba 0.4995 0.1319 1.5 1>]}}
      Random Seed=0
     )
     Scale Factor=&MakeVector4_FromLinearColor(LinearColor=$System.TOD)
    )
    Lifetime Max=1.5
    Lifetime Min=0.9
    Lifetime Mode=::NewEnumerator1
    Sprite Size Mode=::NewEnumerator1
    Uniform Sprite Size Max=3
    Uniform Sprite Size Min=1
    UsePositionOffset=false
    Write Material Random=true
   )
   &ShapeLocation(MaskSpawnGroup=false, Override Seed=false, Sphere Radius=2, UseOwnerScale=true) {enabled=false}
   &AddVelocity(
    Cone Angle=20
    Cone Axis=<0 0 1>
    Module.UseConeFalloff=false
    Module.UseVelDistribution=false
    Override Seed=false
    Rotation Coordinate Space=::Simulation
    Velocity Mode=::NewEnumerator2
    Velocity Speed=&RandomRangeFloat(Minimum=200, Maximum=400)
   )
   &AddVelocity(Velocity=<0 -150 0>) {enabled=false, module=AddVelocity001}
  ]
  particleUpdateScript [
   &Drag(Drag=3, Use Linear Drag=true, Use Rotational Drag=true)
   &GravityForce(Gravity=<0 0 -400>)
   &CurlNoiseForce(Noise Frequency=20, Noise Strength=200, Pan Noise=false)
   &SolveForcesAndVelocity(Clamp Velocity=false, Limit Acceleration=false)
   &ParticleState()
   &ScaleSpriteSize(Uniform Curve Sprite Scale=DataInterface() {propertyValues {Curve {keys [
        {interpMode=RCIM_Cubic, leaveTangent=6.8368, tangentMode=RCTM_Break, time=0, value=0}
        <at 0.2 1>
        {interpMode=RCIM_Cubic, leaveTangent=-5.285, tangentMode=RCTM_Break, time=0.5, value=1}
        {tangentMode=RCTM_User, time=1, value=0}
       ]}}})
  ]
  rendererClasses [&NiagaraSpriteRendererProperties]
  renderers [{rendererClass=&NiagaraSpriteRendererProperties}]
 }
 ConfettiStreak {
  emitterUpdateScript [
   &EmitterState()
   &SpawnBurst_Instantaneous(Spawn Count=10, Use Spawn Probability=false, UseLoopCountLimit=false)
  ]
  particleSpawnScript [
   &InitializeParticle(
    Color=&MultiplyLinearColor(
     Linear Color=&SelectLinearColorFromArray(
      Array Sampling Mode=::NewEnumerator0
      Direct Array Index=?
      Color Selection Array=DataInterface() {propertyValues {ColorData [<rgba 0.849 0.849 0.849 1>, <rgba 0.1674 0.4847 2 1>, <rgba 0.4995 0.1319 1.5 1>]}}
      Random Seed=0
     )
     Scale Factor=&MakeVector4_FromLinearColor(LinearColor=$System.TOD)
    )
    Lifetime Max=0.15
    Lifetime Min=0.07
    Lifetime Mode=::NewEnumerator1
    Sprite Size Mode=::NewEnumerator1
    Uniform Sprite Size Max=5
    Uniform Sprite Size Min=2
    UsePositionOffset=false
    Write Material Random=true
   )
   &ShapeLocation(MaskSpawnGroup=false, Override Seed=false, Sphere Radius=5, UseOwnerScale=true)
   &AddVelocity(
    Cone Angle=30
    Cone Axis=<0 0 1>
    Inner Cone Angle=15
    Module.UseConeFalloff=false
    Module.UseVelDistribution=false
    Override Seed=false
    Rotation Coordinate Space=::Simulation
    Velocity Mode=::NewEnumerator2
    Velocity Speed=&RandomRangeFloat(Minimum=250, Maximum=600)
   )
  ]
  particleUpdateScript [
   &Drag(Drag=10, Use Linear Drag=true, Use Rotational Drag=true)
   &SolveForcesAndVelocity(Clamp Velocity=false, Limit Acceleration=false)
   &ParticleState()
   &ScaleSpriteSizeBySpeed(Max Scale Factor=<1 10>, Min Scale Factor=<0.4 4>)
   &ScaleSpriteSize(Uniform Curve Sprite Scale=DataInterface() {propertyValues {Curve {keys [
        {interpMode=RCIM_Cubic, leaveTangent=6.8368, tangentMode=RCTM_Break, time=0, value=0}
        <at 0.2 1>
        {interpMode=RCIM_Cubic, leaveTangent=-4.4769, tangentMode=RCTM_Break, time=0.6, value=1}
        {tangentMode=RCTM_User, time=1.0114, value=0.0706}
       ]}}})
  ]
  rendererClasses [&NiagaraSpriteRendererProperties]
  renderers [{rendererClass=&NiagaraSpriteRendererProperties}]
 }
 Sparks {
  emitterUpdateScript [
   &EmitterState()
   &SpawnBurst_Instantaneous(
    Spawn Count=&RandomRangeInt(Minimum=3, Maximum=5)
    Use Spawn Probability=false
    UseLoopCountLimit=false
   )
  ]
  particleSpawnScript [
   &InitializeParticle(
    Color=&MultiplyLinearColor(
     Linear Color=&SelectLinearColorFromArray(
      Array Sampling Mode=::NewEnumerator0
      Direct Array Index=?
      Color Selection Array=DataInterface() {propertyValues {ColorData [<rgba 2.4999 2.2397 3 0.8>, <rgba 0.1737 0.7073 4 0.7>, <rgba 1.0253 0.1318 3 0.7>]}}
      Random Seed=0
     )
     Scale Factor=&MakeVector4_FromLinearColor(LinearColor=$System.TOD)
    )
    Lifetime Max=&RandomRangeFloat(Minimum=1, Maximum=1.4)
    Lifetime Min=0.7
    Lifetime Mode=::NewEnumerator1
    Sprite Size Mode=::NewEnumerator1
    Uniform Sprite Size Max=30
    Uniform Sprite Size Min=10
    UsePositionOffset=false
    Write Material Random=true
   )
   &ShapeLocation(MaskSpawnGroup=false, Override Seed=false, Sphere Radius=10, UseOwnerScale=true)
   &AddVelocity(
    Cone Angle=35
    Cone Axis=<0 0 1>
    Module.UseConeFalloff=false
    Module.UseVelDistribution=false
    Override Seed=false
    Rotation Coordinate Space=::Simulation
    Velocity Mode=::NewEnumerator2
    Velocity Speed=&RandomRangeFloat(Minimum=350, Maximum=500)
   )
  ]
  particleUpdateScript [
   &ParticleState()
   &ScaleSpriteSize(Uniform Curve Sprite Scale=DataInterface() {propertyValues {Curve {keys [
        {interpMode=RCIM_Cubic, tangentMode=RCTM_Break, time=0, value=0}
        {interpMode=RCIM_Cubic, tangentMode=RCTM_Break, time=0.1, value=1}
        {interpMode=RCIM_Cubic, tangentMode=RCTM_Break, time=0.2, value=0}
        {arriveTangent=5.8511, interpMode=RCIM_Cubic, leaveTangent=-6.2653, tangentMode=RCTM_Break, time=0.3023, value=0.9569}
        {interpMode=RCIM_Cubic, tangentMode=RCTM_Break, time=0.4029, value=0}
        {arriveTangent=7.3852, interpMode=RCIM_Cubic, leaveTangent=-4.8986, tangentMode=RCTM_Break, time=0.5198, value=0.7267}
        {interpMode=RCIM_Cubic, tangentMode=RCTM_Break, time=0.6539, value=0}
        {interpMode=RCIM_Cubic, leaveTangent=-5.1375, tangentMode=RCTM_Break, time=0.7904, value=0.4154}
        {interpMode=RCIM_Cubic, tangentMode=RCTM_Break, time=0.9977, value=0}
       ]}}})
   &GravityForce(Gravity=&Multiply_VectorByFloat(
     Vector=&RandomRangeVector(Minimum=<0 0 -350>, Maximum=<0 0 -650>)
     Float=&FloatFromCurve(FloatCurve=DataInterface() {propertyValues {Curve {keys [<at 0 1>, {interpMode=RCIM_Cubic, leaveTangent=-3.9842, tangentMode=RCTM_Break, time=0.3, value=1}, <at 0.5 0.2>, <at 1 0>]}}})
    ))
   &CurlNoiseForce(Noise Strength=150, Pan Noise=false)
   &Drag(Drag=6, Use Linear Drag=true, Use Rotational Drag=true)
   &SolveForcesAndVelocity(Clamp Velocity=false, Limit Acceleration=false)
   &ScaleColor(
    Scale Alpha=&FloatFromCurve(FloatCurve=DataInterface() {propertyValues {Curve {keys [<at 0 1>,<at 0.7 0.9>,<at 1 0>]}}})
    ScaleA=true
    ScaleRGB=true
   )
  ]
  rendererClasses [&NiagaraSpriteRendererProperties]
  renderers [{rendererClass=&NiagaraSpriteRendererProperties}]
 }
 SquareConfetti {
  emitterUpdateScript [
   &EmitterState()
   &SpawnBurst_Instantaneous(Spawn Count=12, Use Spawn Probability=false, UseLoopCountLimit=false)
  ]
  particleSpawnScript [
   &InitializeParticle(
    Color=&MultiplyLinearColor(
     Linear Color=&SelectLinearColorFromArray(
      Array Sampling Mode=::NewEnumerator0
      Direct Array Index=?
      Color Selection Array=DataInterface() {propertyValues {ColorData [<rgba 0.849 0.849 0.849 1>, <rgba 0.1674 0.4847 2 1>, <rgba 0.4995 0.1319 1.5 1>]}}
      Random Seed=0
     )
     Scale Factor=&MakeVector4_FromLinearColor(LinearColor=$System.TOD)
    )
    Lifetime Max=1.25
    Lifetime Min=0.5
    Lifetime Mode=::NewEnumerator1
    Sprite Size Mode=::NewEnumerator1
    Uniform Sprite Size Max=6
    Uniform Sprite Size Min=3
    UsePositionOffset=false
    Write Material Random=true
   )
   /Game/Effects/Niagara/Modules/Sprite/FN_SpriteUV()
   &ShapeLocation(MaskSpawnGroup=false, Override Seed=false, Sphere Radius=4, UseOwnerScale=true)
   &AddVelocity(
    Cone Angle=35
    Cone Axis=<0 0 1>
    Module.UseConeFalloff=false
    Module.UseVelDistribution=false
    Override Seed=false
    Rotation Coordinate Space=::Simulation
    Velocity Mode=::NewEnumerator2
    Velocity Speed=&RandomRangeFloat(Minimum=250, Maximum=600)
   )
   &AddVelocity(Rotation Coordinate Space=::Simulation, Velocity=<0 -100 0>) {enabled=false, module=AddVelocity001}
  ]
  particleUpdateScript [
   &Drag(Drag=6, Use Linear Drag=true, Use Rotational Drag=true)
   &GravityForce(Gravity=&Multiply_VectorByFloat(
     Vector=&RandomRangeVector(Minimum=<0 0 -500>, Maximum=<0 0 -800>)
     Float=&FloatFromCurve(FloatCurve=DataInterface() {propertyValues {Curve {keys [<at 0 1>, {interpMode=RCIM_Cubic, leaveTangent=-3.9842, tangentMode=RCTM_Break, time=0.3, value=1}, <at 0.5 0.2>, <at 1 0>]}}})
    ))
   &CurlNoiseForce(Noise Strength=150, Pan Noise=false)
   &SolveForcesAndVelocity(Clamp Velocity=false, Limit Acceleration=false)
   &ParticleState()
   &ScaleSpriteSize(Uniform Curve Sprite Scale=DataInterface() {propertyValues {Curve {keys [
        {interpMode=RCIM_Cubic, leaveTangent=6.8368, tangentMode=RCTM_Break, time=0, value=0}
        <at 0.2 1>
        {interpMode=RCIM_Cubic, leaveTangent=-6.915, tangentMode=RCTM_Break, time=0.6, value=1}
        {tangentMode=RCTM_User, time=1.0114, value=0}
       ]}}})
   &SubUVAnimation(
    "Sprite Renderer "=DataInterface() {propertyValues {SpriteRenderer=&self:SquareConfetti_0.NiagaraSpriteRendererProperties_0}}
    UseEndFrame=false
    UseStartFrame=false
   )
  ]
  rendererClasses [&NiagaraSpriteRendererProperties]
  renderers [{rendererClass=&NiagaraSpriteRendererProperties}]
 }
}
summary {
 systemName=NS_Emote_Glitter
 emitters [
  {emitterName=SquareConfetti, rendererClasses [&NiagaraSpriteRendererProperties]}
  {emitterName=ConfettiBurst, simTarget=GPUComputeSim, rendererClasses [&NiagaraSpriteRendererProperties]}
  {emitterName=ConfettiStreak, rendererClasses [&NiagaraSpriteRendererProperties]}
  {emitterName=ConfettiResidual, rendererClasses [&NiagaraSpriteRendererProperties]}
  {emitterName=BurstMedLowOnly, rendererClasses [&NiagaraSpriteRendererProperties]}
  {emitterName=Sparks, rendererClasses [&NiagaraSpriteRendererProperties]}
 ]
 userVariables []
}
